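# Physics asset for the male medium Handyman backpack cosmetic, defining collision and joint limits for its dangling grenades and shotgun shells.
# Capsule bodies cover the pack, two grenade bones and six shell bones, each of the latter eight constrained directly to the pack.
# Grenades get wide swing and moderate twist limits, while the shells are held to a few degrees so they move only slightly.
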PhysicsAsset /BRCosmetics/Accessories/FORT_Backpacks/Backpack_M_MED_Handyman/Meshes/M_MED_Handyman_Pack_Physics
bodies {
 dyn_grenadeBottom [{shapeType=Capsule, length=6, radius=3, rotation=<rot 0 0 90>}]
 dyn_grenadeTop [{shapeType=Capsule, length=6, radius=3, rotation=<rot 0 0 90>}]
 dyn_shotgunShell1 [{shapeType=Capsule, length=3, radius=1, rotation=<rot 0 0 90>}]
 dyn_shotgunShell2 [{shapeType=Capsule, length=3, radius=1, rotation=<rot 0 0 90>}]
 dyn_shotgunShell3 [{shapeType=Capsule, length=3, radius=1, rotation=<rot 0 0 90>}]
 dyn_shotgunShell4 [{shapeType=Capsule, length=3, radius=1, rotation=<rot 0 0 90>}]
 dyn_shotgunShell5 [{shapeType=Capsule, length=3, radius=1, rotation=<rot 0 0 90>}]
 dyn_shotgunShell6 [{shapeType=Capsule, length=3, radius=1, rotation=<rot 0 0 90>}]
 pack [{shapeType=Capsule, length=6.2198, radius=3.4762, center=<0 -2.0741 -0.0756>}]
}
constraints [
 {bone1Name=dyn_grenadeTop, bone2Name=pack, swing1LimitDegrees=45, swing2LimitDegrees=45, twistLimitDegrees=15}
 {bone1Name=dyn_grenadeBottom, bone2Name=pack, swing1LimitDegrees=45, swing2LimitDegrees=45, twistLimitDegrees=15}
 {bone1Name=dyn_shotgunShell1, bone2Name=pack, swing1LimitDegrees=5, swing2LimitDegrees=5, twistLimitDegrees=2}
 {bone1Name=dyn_shotgunShell2, bone2Name=pack, swing1LimitDegrees=5, swing2LimitDegrees=5, twistLimitDegrees=2}
 {bone1Name=dyn_shotgunShell3, bone2Name=pack, swing1LimitDegrees=5, swing2LimitDegrees=5, twistLimitDegrees=2}
 {bone1Name=dyn_shotgunShell4, bone2Name=pack, swing1LimitDegrees=5, swing2LimitDegrees=5, twistLimitDegrees=2}
 {bone1Name=dyn_shotgunShell5, bone2Name=pack, swing1LimitDegrees=5, swing2LimitDegrees=5, twistLimitDegrees=2}
 {bone1Name=dyn_shotgunShell6, bone2Name=pack, swing1LimitDegrees=5, swing2LimitDegrees=5, twistLimitDegrees=2}
]
